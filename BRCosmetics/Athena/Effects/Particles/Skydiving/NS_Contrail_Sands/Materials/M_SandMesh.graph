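# M_SandMesh is an additive, two-sided Niagara particle material for the sand skydiving contrail.
# A screen-aligned water-noise normal texture, offset by particle UVs and animated over time, indexes a sand color curve atlas, tinted by particle color for base color and emissive.
# An 8x8 SubUV smoke swirl, minus a per-cell sphere mask, multiplies a hard noise cutout driven by a dynamic parameter and particle alpha to form opacity.
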Material /BRCosmetics/Athena/Effects/Particles/Skydiving/NS_Contrail_Sands/Materials/M_SandMesh
graph {
 properties {
  bEnableResponsiveAA=true
  bUseMaterialAttributes=true
  bUsedWithNiagaraSprites=true
  blendMode=BLEND_Additive
  pixelDepthOffsetMode=PDOM_Legacy
  twoSided=true
  usedShadingModels=Default Lit
 }
 nodes [
  n1 = TextureCoordinate()
  n2 = Frac(None=TextureCoordinate() {uTiling=8, vTiling=8})
  n3 = Constant() {r=0.5}
  n4 = Constant() {r=0.2}
  n5 = Constant()
  n6 = Saturate(None=Subtract(
    A=ParticleSubUV!(UVs=@n1, Tex, Apply View MipBias) {texture=/Game/Effects/Fort_Effects/Textures/Smoke/T_SmokeSwirl_SubUV_Simple_8x8}.R
    B=SphereMask(A=@n2, B=@n3, Radius=@n4, Hardness=@n5)
   ))
  Ceil(None=@n6)
  MaterialFunctionCall() {materialFunction=&ScreenAlignedUVs}
  n7 = Constant() {r=24}
  n8 = TextureCoordinate() {uTiling=8, vTiling=8}
  n9 = TextureSample(UVs=Add(
    A=MaterialFunctionCall!(TextureResolution=@n7) {materialFunction=/Engine/Functions/Engine_MaterialFunctions02/Texturing/ScreenAlignedPixelToPixelUVs}.Result
    B=Multiply(A=@n8) {constB=0.15}
   )) {samplerType=SAMPLERTYPE_Normal, texture=/Game/Environments/World/Backgrounds/Locales/Temperate/Textures/T_WaterNoise}
  n10 = Multiply(A=ConstantBiasScale(None=@n9.RGB), B=Time())
  n11 = CurveAtlasRowParameter(CurveTime=MaterialFunctionCall(Float3=Frac(None=@n10)) {materialFunction=&BreakOutFloat3Components}.R) {
   atlas=&AthenaCurveAtlas01
   curve=/Game/Athena/LinearColorCurves/Curve_Sand_EmissiveColor
   defaultValue=43
   parameterName=Emissive
  }
  n12 = Multiply(A=@n11) {constB=0.1}
  QualitySwitch(Default=@n12, Low=@n12, High=@n12, Medium=@n12)
  n13 = Multiply(A=ParticleColor().RGB, B=@n12)
  n14 = DynamicParameter() {defaultValue=<rgba 1 0 0 1>}
  n15 = MaterialFunctionCall(R=@n14.R, G=@n14.R, B=@n14.R, A=@n14.R) {materialFunction=&AppendMany}
  n16 = Ceil(None=Saturate(None=DotProduct(A=@n9.RGB, B=@n15.RGB)))
  SetMaterialAttributes!(
   MaterialAttributes=SetMaterialAttributes!(MaterialAttributes, World Position Offset) {attributeSetTypes [#1]}
   Emissive Color=@n13
   Opacity=Multiply(A=Multiply(A=@n16, B=@n6), B=ParticleColor().A)
   Base Color=@n13
   Normal
  ) {attributeSetTypes [
    #2
    #3
    #4
    #5
   ]}
 ]
}
