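# Niagara skydiving contrail cosmetic "Sands", with a clock-themed, warm gold look trailing from bones of the user-supplied skeletal mesh.
# Sprite emitters layer clocks, dark gradient, fog, flipbooks and drifting Roman numerals, picking bones by spawn group, alongside a light emitter.
# Orange-yellow ribbons take their added velocity scale from a system integrator value alongside the skydiving attribute modules.
# The combined multi-renderer Sprites emitter is disabled, as are several ribbon offset and velocity modules.
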NiagaraSystem /BRCosmetics/Athena/Effects/Particles/Skydiving/NS_Contrail_Sands/NS_Contrail_Sands
dependencies {
 usedDataInterfaces [
  &NiagaraDataInterfaceCurve
  &NiagaraDataInterfaceVectorCurve
  &NiagaraDataInterfaceColorCurve
  &NiagaraDataInterfaceVector2DCurve
 ]
 usedDynamicInputs [&FN_FrontEndFloatModifier, &SetBoolByFloatComparison]
 usedModules [
  &SystemState
  /Game/Effects/Niagara/Modules/Skydiving/FN_DivingAttributes/v5/FN_DivingAttributes
  &self:SystemSpawnScript.SystemScriptSource.SystemScriptGraph.NiagaraNodeAssignment_0.SetVariables_#1
  &FN_Integrator
 ]
}
emitters {
 Clock {
  emitterUpdateScript [
   &EmitterState(Inactive Response=::NewEnumerator1, Life Cycle Mode=::NewEnumerator1, Loop Duration=1.3, UseLoopDelay=true)
   &SpawnRate(SpawnRate=0, Use Spawn Probability=false)
   &SpawnBurst_Instantaneous(Spawn Count=4, Use Spawn Probability=false)
  ]
  particleSpawnScript [
   &FN_SetSpawnGroup(Number of Spawn Groups=2)
   /Niagara/Modules/Spawn/Initialization/InitializeParticle(
    Lifetime=2
    Module.Write Color=true
    Module.Write Lifetime=true
    Module.Write Mass=true
    Module.Write Material Random=true
    Module.Write Position=false
    Module.Write Scale=false
    Module.Write SpriteRotation=false
    Module.Write SpriteSize=true
    Sprite Size=&RandomRangeVector2D(Minimum=<48 48>, Maximum=<50 50>, Evaluation Type=::NewEnumerator1)
   )
   &DynamicMaterialParameters(
    Index 0 Param 2=&RandomRangeFloat(Minimum=0, Maximum=0, Evaluation Type=::NewEnumerator1)
    Index 0 Param 3=0.1
    Module.Param0WriteEnabled=false
    Module.Param1WriteEnabled=true
    Module.Param2WriteEnabled=true
    Module.Param3WriteEnabled=true
   ) {module=DynamicMaterialParameters001}
   &CameraOffset(Camera Offset Amount=&RandomRangeFloat(Minimum=25, Maximum=50, Evaluation Type=::NewEnumerator1))
  ]
  particleUpdateScript [
   &ParticleState()
   &Color(Color=&MakeLinearColorFromVectorAndFloat(
     "Vector (RGB)"=&VectorFromCurve(VectorCurve=DataInterface() {propertyValues {XCurve {keys [<at 0 1>,<at 1 1>]}, YCurve {keys [<at 0 1>,<at 1 1>]}, ZCurve {keys [<at 0 1>,<at 1 1>]}}})
     "Float (Alpha)"=&ClampFloat(Float=&FloatFromCurve(FloatCurve=DataInterface() {propertyValues {Curve {keys [{arriveTangent=0.0516, interpMode=RCIM_Cubic, time=-0.003, value=0}, <at 0.2978 0.2015 RCIM_Cubic>, <at 1 0>]}}}))
    ))
   &SkeletalMeshLocation(
    Apply Sampled Orientation to Mesh Particles=true
    Bone / Socket Index=$Particles.SpawnGroup
    Bone Sampling Mode=::NewEnumerator4
    Inherit Velocity=false
    Interpolate Mesh Sampling=false
    MaskSpawnGroup=false
    Skeletal Mesh=$User.SkeletalMesh
   ) {module=SkeletalMeshLocation001}
   /Game/Effects/Niagara/Modules/Skydiving/FN_SkydivePositionOffset_v1/FN_SkydivePositionOffset(
    Bone Orientation:Quat4f=$Output.SkeletalMeshLocation001.SampledRotation
    Bone Position:Vector3f=$Output.SkeletalMeshLocation001.SampledPosition
    Offset:NiagaraFloat=/Game/Effects/Niagara/DynamicInputs/Skydiving/SkydiveHandsFeetFloatModifier(Hands:NiagaraFloat=0, Feet:NiagaraFloat=0)
   )
   &ScaleSpriteSize(Scale Factor=&Vector2DFromFloat(Value=&FloatFromCurve(FloatCurve=DataInterface() {propertyValues {Curve {keys [
          {arriveTangent=19.8573, interpMode=RCIM_Cubic, time=0, value=0.3202}
          {arriveTangent=0.5332, interpMode=RCIM_Cubic, leaveTangent=0.5332, time=0.1513, value=0.7545}
          {arriveTangent=-0.4802, interpMode=RCIM_Cubic, leaveTangent=-0.4802, time=0.8198, value=0.7573}
          {interpMode=RCIM_Cubic, leaveTangent=-14.9405, time=1.0022, value=0.3459}
         ]}}})))
   &DynamicMaterialParameters(
    Index 0 Param 1=1
    Index 0 Param 3=0.1
    Index 0 Param 4=&FloatFromCurve(FloatCurve=DataInterface() {propertyValues {Curve {keys [
         {interpMode=RCIM_Cubic, tangentMode=RCTM_User, time=0, value=-0.3}
         {interpMode=RCIM_Cubic, leaveTangent=-0.9983, tangentMode=RCTM_User, time=1, value=-3.5}
        ]}}})
    Module.Param0WriteEnabled=true
    Module.Param1WriteEnabled=false
    Module.Param2WriteEnabled=true
    Module.Param3WriteEnabled=true
   )
  ]
  rendererClasses [&NiagaraSpriteRendererProperties]
  renderers [{rendererClass=&NiagaraSpriteRendererProperties}]
 }
 Clock001 {
  emitterUpdateScript [
   &EmitterState(Inactive Response=::NewEnumerator1, Life Cycle Mode=::NewEnumerator1, Loop Duration=1, UseLoopDelay=true)
   &SpawnRate(SpawnRate=2, Use Spawn Probability=false)
   &SpawnBurst_Instantaneous(Spawn Count=4, Use Spawn Probability=false)
  ]
  particleSpawnScript [
   &FN_SetSpawnGroup(Number of Spawn Groups=2)
   &InitializeParticle(
    Lifetime=2
    Module.Write Color=true
    Module.Write Lifetime=true
    Module.Write Mass=true
    Module.Write Material Random=true
    Module.Write Position=false
    Module.Write Scale=false
    Module.Write SpriteRotation=true
    Module.Write SpriteSize=true
    Sprite Rotation=&NormalizedAngleToDegrees(NormalizedAngle=&RandomRangeFloat(Minimum=0, Maximum=1, Evaluation Type=::NewEnumerator1))
    Sprite Size=&RandomRangeVector2D(Minimum=<50 50>, Maximum=<70 70>, Evaluation Type=::NewEnumerator1)
   )
   &DynamicMaterialParameters(Module.Param0WriteEnabled=false, Module.Param1WriteEnabled=true, Module.Param2WriteEnabled=false, Module.Param3WriteEnabled=false) {module=DynamicMaterialParameters001}
   &CameraOffset(Camera Offset Amount=&RandomRangeFloat(Minimum=25, Maximum=50, Evaluation Type=::NewEnumerator1))
  ]
  particleUpdateScript [
   &ParticleState()
   &Color(Color=&MakeLinearColorFromVectorAndFloat(
     "Vector (RGB)"=<1 0.4904 0.0776>
     "Float (Alpha)"=&ClampFloat(Float=&FloatFromCurve(
       FloatCurve=DataInterface() {propertyValues {Curve {keys [{arriveTangent=0.0516, interpMode=RCIM_Cubic, time=-0.003, value=0}, <at 0.2978 0.2468 RCIM_Cubic>, <at 1 0>]}}}
       Scale Curve=0.5
      ))
    ))
   &SkeletalMeshLocation(
    Apply Sampled Orientation to Mesh Particles=true
    Bone / Socket Index=$Particles.SpawnGroup
    Bone Sampling Mode=::NewEnumerator4
    Inherit Velocity=false
    Interpolate Mesh Sampling=false
    MaskSpawnGroup=false
    Skeletal Mesh=$User.SkeletalMesh
   ) {module=SkeletalMeshLocation001}
   &FN_SkydivePositionOffset(
    Bone Orientation:Quat4f=$Output.SkeletalMeshLocation001.SampledRotation
    Bone Position:Vector3f=$Output.SkeletalMeshLocation001.SampledPosition
    Offset:NiagaraFloat=&SkydiveHandsFeetFloatModifier(Hands:NiagaraFloat=0, Feet:NiagaraFloat=0)
   )
   &DynamicMaterialParameters(
    Index 0 Param 1=1
    Index 0 Param 3=0.1
    Index 0 Param 4=&FloatFromCurve(FloatCurve=DataInterface() {propertyValues {Curve {keys [
         {arriveTangent=-5.637, interpMode=RCIM_Cubic, time=0, value=0.9968}
         {arriveTangent=-0.9983, interpMode=RCIM_Cubic, leaveTangent=-0.9983, time=0.3196, value=0.0047}
         <at 0.9984 0 RCIM_Cubic>
        ]}}})
    Module.Param0WriteEnabled=true
    Module.Param1WriteEnabled=false
    Module.Param2WriteEnabled=true
    Module.Param3WriteEnabled=true
   )
  ]
  rendererClasses [&NiagaraSpriteRendererProperties]
  renderers [{rendererClass=&NiagaraSpriteRendererProperties}]
 }
 DarkGradient {
  emitterUpdateScript [
   &EmitterState(Inactive Response=::NewEnumerator1, Life Cycle Mode=::NewEnumerator1, Loop Duration=1.3, UseLoopDelay=true)
   &SpawnRate(SpawnRate=0, Use Spawn Probability=false)
   &SpawnBurst_Instantaneous(Spawn Count=4, Use Spawn Probability=false)
  ]
  particleSpawnScript [
   &FN_SetSpawnGroup()
   &InitializeParticle(
    Lifetime=&RandomRangeFloat(Minimum=1.3, Maximum=1.3, Evaluation Type=::NewEnumerator1)
    Module.Write Color=true
    Module.Write Lifetime=true
    Module.Write Mass=true
    Module.Write Material Random=true
    Module.Write Position=false
    Module.Write Scale=false
    Module.Write SpriteRotation=false
    Module.Write SpriteSize=true
    Sprite Size=<80 80>
   )
   &Color(Color=&MakeLinearColorFromVectorAndFloat("Vector (RGB)"=<0.1719 0.1133 0.0358>, "Float (Alpha)"=&ClampFloat(Float=1.2)))
  ]
  particleUpdateScript [
   &ParticleState()
   &SkeletalMeshLocation(
    Apply Sampled Orientation to Mesh Particles=true
    Bone / Socket Index=$Particles.SpawnGroup
    Bone Sampling Mode=::NewEnumerator4
    Inherit Velocity=false
    Interpolate Mesh Sampling=false
    MaskSpawnGroup=false
    Skeletal Mesh=$User.SkeletalMesh
   ) {module=SkeletalMeshLocation001}
   &ScaleColor(
    Scale Alpha=&FloatFromCurve(
     FloatCurve=DataInterface() {propertyValues {Curve {keys [
         <at 0 0 RCIM_Cubic>
         {arriveTangent=0.9612, interpMode=RCIM_Cubic, leaveTangent=0.9612, time=0.1158, value=0.8491}
         {arriveTangent=-0.9602, interpMode=RCIM_Cubic, leaveTangent=-0.9602, time=0.8838, value=0.8495}
         <at 1 0>
        ]}}}
     Scale Curve=0.6
    )
    ScaleA=true
    ScaleRGB=true
   )
   &DynamicMaterialParameters(
    Index 0 Param 1=2
    Index 0 Param 2=1
    Index 0 Param 3=1
    Index 0 Param 4=1
    Module.Param0WriteEnabled=true
    Module.Param1WriteEnabled=true
    Module.Param2WriteEnabled=true
    Module.Param3WriteEnabled=true
   )
   &CameraOffset(Camera Offset Amount=15)
   &ScaleSpriteSize(Scale Factor=&Vector2DFromFloat(Value=&FloatFromCurve(FloatCurve=DataInterface() {propertyValues {Curve {keys [<at 0 0.8>,<at 1 1.25>]}}})))
  ]
  rendererClasses [&NiagaraSpriteRendererProperties]
  renderers [{rendererClass=&NiagaraSpriteRendererProperties}]
 }
 FN_Ribbons {
  emitterUpdateScript [
   &EmitterState()
   /Game/Effects/Niagara/Modules/Skydiving/FN_DivingAttributes/FN_EmitterDivingAttributes_v1(SkeletalMesh:NiagaraDataInterfaceSkeletalMesh=$User.SkeletalMesh)
   &FN_SpawnRibbon_v2(Number of Ribbons=2, Spawn Mode=::NewEnumerator0, SpawnRate=30)
  ]
  particleSpawnScript [
   &FN_InitializeRibbon_v3(
    Color:LinearColor=/Niagara/DynamicInputs/TypeConversions/MakeLinearColorFromFloat(Value:NiagaraFloat=0.6)
    Enable Skeletal Position=true
    Lifetime=0.35
    Ribbon Width=50
    Sampled Position Coordinate Space:ENiagaraCoordinateSpace=::Local
    Sampled Rotation Coordinate Space:ENiagaraCoordinateSpace=::Local
    Sampled Velocity Coordinate Space:ENiagaraCoordinateSpace=::Local
   ) {module=FN_InitializeRibbon_v3001}
   &AddVelocity(
    Scale Added Velocity=&Lerp_Vector(A=<1 0 1>, B=<1 1 2>, Alpha=$System.FN_Integrator001.Value)
    Velocity=<0 -600 500>
   )
   &SkeletalMeshLocation(
    Apply Sampled Orientation to Mesh Particles=true
    Apply To Particle Position=false
    Bone / Socket Index=$Particles.SpawnGroup
    Bone Sampling Mode=::NewEnumerator4
    Inherit Velocity=false
    MaskSpawnGroup=false
    Skeletal Mesh=$User.SkeletalMesh
   ) {enabled=false, module=SkeletalMeshLocation001}
   &SampleSkeletalMeshSkeleton(Interpolate Bone Sampling=false, Mesh:NiagaraDataInterfaceSkeletalMesh=$User.SkeletalMesh)
   /Game/Effects/Niagara/Modules/Skydiving/FN_Skydive_Velocity_v4/FN_Skydive_Velocity(
    Bone Orientation:Quat4f=$Output.SkeletalMeshLocation001.SampledRotation
    Bone Position:Vector3f=$Output.SkeletalMeshLocation001.SampledPosition
    FE Speed=500
    FE Spread:NiagaraFloat=&SkydiveHandsFeetFloatModifier(Hands:NiagaraFloat=&Lerp_Float(A=0, B=200, Alpha=$System.Dive), Feet:NiagaraFloat=0)
    SpawnGroup:NiagaraInt32=$Particles.SpawnGroup
    Speed=221.3639
    Spread=148.0048
   ) {enabled=false, module=FN_Skydive_Velocity001}
   &FN_SkydivePositionOffset(
    Bone Orientation:Quat4f=$Output.SkeletalMeshLocation001.SampledRotation
    Bone Position:Vector3f=$Output.SkeletalMeshLocation001.SampledPosition
    Offset:NiagaraFloat=&SkydiveHandsFeetFloatModifier(Hands:NiagaraFloat=25, Feet:NiagaraFloat=50)
   ) {enabled=false}
  ]
  particleUpdateScript [
   &SolveForcesAndVelocity(Clamp Velocity=false, Limit Acceleration=false)
   &FN_UpdateRibbon(
    Enable Skeletal Position=true
    Initial Color:LinearColor=$Particles.Initial.Color
    Initial RibbonWidth:NiagaraFloat=$Particles.Initial.RibbonWidth
   )
   &Color(
    Color=&ColorFromCurve(ColorCurve=DataInterface() {propertyValues {
       AlphaCurve {keys [<at 0 1>]}
       BlueCurve {keys [<at 0 0>, <at 0.1935 0.0359>, {arriveTangent=-0.1293, interpMode=RCIM_Cubic, leaveTangent=-0.1293, time=0.2445, value=0}, <at 0.4709 0>]}
       GreenCurve {keys [
         <at 0 0.5462>
         <at 0.1935 0.6179>
         {arriveTangent=-0.7253, interpMode=RCIM_Cubic, leaveTangent=-0.7253, time=0.2445, value=0.5833}
         <at 0.4709 0.4167>
        ]}
       RedCurve {keys [<at 0 1>,<at 0.1935 1>,<at 0.2445 1 RCIM_Cubic>,<at 0.4709 1>]}
      }})
    Scale Alpha=&FloatFromCurve(
     FloatCurve=DataInterface() {propertyValues {Curve {keys [
         {arriveTangent=0.7543, interpMode=RCIM_Cubic, leaveTangent=0.7543, tangentMode=RCTM_User, time=0, value=0}
         <at 0.1 0.05>
         {arriveTangent=-0.013, interpMode=RCIM_Cubic, leaveTangent=-0.013, tangentMode=RCTM_User, time=0.2992, value=0.0096}
         {interpMode=RCIM_Cubic, leaveTangent=-0.2611, time=1, value=0}
        ]}}}
     Scale Curve=5
    )
   )
   &Material(Base Texture Time:NiagaraFloat=&Multiply_Float(A=$Emitter.Age, B=-1), PostMult=0.5455, Power=0.4408)
  ]
  rendererClasses [&NiagaraRibbonRendererProperties]
  renderers [{rendererClass=&NiagaraRibbonRendererProperties}]
 }
 Fog {
  emitterUpdateScript [
   &EmitterState(Inactive Response=::NewEnumerator1, Life Cycle Mode=::NewEnumerator1, Loop Duration=1, UseLoopDelay=true)
   &SpawnRate(SpawnRate=8, Use Spawn Probability=false)
   &SpawnBurst_Instantaneous(Spawn Count=4, Use Spawn Probability=false)
  ]
  particleSpawnScript [
   &InitializeParticle(
    Lifetime=&RandomRangeFloat(Minimum=1, Maximum=1.5, Evaluation Type=::NewEnumerator1)
    Module.Write Color=true
    Module.Write Lifetime=true
    Module.Write Mass=true
    Module.Write Material Random=true
    Module.Write Position=false
    Module.Write Scale=false
    Module.Write SpriteRotation=true
    Module.Write SpriteSize=true
    Sprite Rotation=&NormalizedAngleToDegrees(NormalizedAngle=&RandomRangeFloat(Minimum=0, Maximum=1, Evaluation Type=::NewEnumerator1))
    Sprite Size=&Vector2DFromFloat(Value=&RandomRangeFloat(Minimum=25, Maximum=25, Evaluation Type=::NewEnumerator1))
   )
   &AddVelocity(Velocity=&RandomRangeVector(Minimum=<-10 -10 50>, Maximum=<10 10 100>, Evaluation Type=::NewEnumerator1))
   &FN_SetSpawnGroup()
  ]
  particleUpdateScript [
   &ParticleState()
   &SolveForcesAndVelocity(Clamp Velocity=false, Limit Acceleration=false)
   &Color(Color=&MakeLinearColorFromVectorAndFloat(
     "Vector (RGB)"=&VectorFromCurve(VectorCurve=DataInterface() {propertyValues {XCurve {keys [<at 0 2>,<at 1 15>]}, YCurve {keys [<at 0 1.7922>,<at 1 11.4004>]}, ZCurve {keys [<at 0 1.1042>,<at 1 3.8281>]}}})
     "Float (Alpha)"=&ClampFloat(Float=&FloatFromCurve(
       FloatCurve=DataInterface() {propertyValues {Curve {keys [
           <at -0.0004 0 RCIM_Cubic>
           {arriveTangent=1.8811, interpMode=RCIM_Cubic, leaveTangent=1.8811, time=0.1829, value=0.9903}
           {arriveTangent=-1, leaveTangent=-1, time=0.5313, value=1}
           <at 1 0>
          ]}}}
       Scale Curve=0.02
      ))
    ))
   &SkeletalMeshLocation(
    Apply Sampled Orientation to Mesh Particles=true
    Bone / Socket Index=$Particles.SpawnGroup
    Bone Sampling Mode=::NewEnumerator4
    Inherit Velocity=false
    Interpolate Mesh Sampling=false
    MaskSpawnGroup=false
    Skeletal Mesh=$User.SkeletalMesh
   ) {module=SkeletalMeshLocation001}
   &FN_SkydivePositionOffset(
    Bone Orientation:Quat4f=$Output.SkeletalMeshLocation001.SampledRotation
    Bone Position:Vector3f=$Output.SkeletalMeshLocation001.SampledPosition
    Offset:NiagaraFloat=&SkydiveHandsFeetFloatModifier(Hands:NiagaraFloat=0, Feet:NiagaraFloat=0)
   )
   /Niagara/Modules/Update/SubUV/SubUVAnimation(Number Of Frames=54.7772)
   &ScaleSpriteSize(Scale Factor=&Vector2DFromFloat(Value=&FloatFromCurve(FloatCurve=DataInterface() {propertyValues {Curve {keys [<at 0 1>,<at 1 5>]}}})))
   &CameraOffset(Camera Offset Amount=25)
  ]
  rendererClasses [&NiagaraSpriteRendererProperties]
  renderers [{rendererClass=&NiagaraSpriteRendererProperties}]
 }
 FogCenter {
  emitterUpdateScript [
   &EmitterState(Inactive Response=::NewEnumerator1, Life Cycle Mode=::NewEnumerator1, Loop Duration=1, UseLoopDelay=true)
   &SpawnRate(SpawnRate=8, Use Spawn Probability=false)
   &SpawnBurst_Instantaneous(Spawn Count=4, Use Spawn Probability=false)
  ]
  particleSpawnScript [
   &FN_SetSpawnGroup()
   &InitializeParticle(
    Lifetime=&RandomRangeFloat(Minimum=1, Maximum=1.5, Evaluation Type=::NewEnumerator1)
    Module.Write Color=true
    Module.Write Lifetime=true
    Module.Write Mass=true
    Module.Write Material Random=true
    Module.Write Position=false
    Module.Write Scale=false
    Module.Write SpriteRotation=true
    Module.Write SpriteSize=true
    Sprite Rotation=&NormalizedAngleToDegrees(NormalizedAngle=&RandomRangeFloat(Minimum=0, Maximum=1, Evaluation Type=::NewEnumerator1))
    Sprite Size=&RandomRangeVector2D(Minimum=<15 15>, Maximum=<25 25>, Evaluation Type=::NewEnumerator1)
   )
   &AddVelocity(Velocity=&RandomRangeVector(Minimum=<-10 -10 50>, Maximum=<10 10 100>, Evaluation Type=::NewEnumerator1))
  ]
  particleUpdateScript [
   &ParticleState()
   &SolveForcesAndVelocity(Clamp Velocity=false, Limit Acceleration=false)
   &Color(Color=&MakeLinearColorFromVectorAndFloat(
     "Vector (RGB)"=&VectorFromCurve(VectorCurve=DataInterface() {propertyValues {XCurve {keys [<at 0 30>,<at 1 20>]}, YCurve {keys [<at 0 25.2878>,<at 1 13.7166>]}, ZCurve {keys [<at 0 9.6875>,<at 1 0.4988>]}}})
     "Float (Alpha)"=&ClampFloat(Float=&FloatFromCurve(
       FloatCurve=DataInterface() {propertyValues {Curve {keys [
           <at -0.0004 0 RCIM_Cubic>
           {arriveTangent=1.8811, interpMode=RCIM_Cubic, leaveTangent=1.8811, time=0.1829, value=0.9903}
           {arriveTangent=-1, leaveTangent=-1, time=0.5313, value=1}
           <at 1 0>
          ]}}}
       Scale Curve=0.05
      ))
    ))
   &SkeletalMeshLocation(
    Apply Sampled Orientation to Mesh Particles=true
    Bone / Socket Index=$Particles.SpawnGroup
    Bone Sampling Mode=::NewEnumerator4
    Inherit Velocity=false
    Interpolate Mesh Sampling=false
    MaskSpawnGroup=false
    Skeletal Mesh=$User.SkeletalMesh
   ) {module=SkeletalMeshLocation001}
   &FN_SkydivePositionOffset(
    Bone Orientation:Quat4f=$Output.SkeletalMeshLocation001.SampledRotation
    Bone Position:Vector3f=$Output.SkeletalMeshLocation001.SampledPosition
    Offset:NiagaraFloat=&SkydiveHandsFeetFloatModifier(Hands:NiagaraFloat=0, Feet:NiagaraFloat=0)
   )
   &SubUVAnimation(Number Of Frames=54.7772)
   &ScaleSpriteSize(Scale Factor=&Vector2DFromCurve(Vector2Curve=DataInterface() {propertyValues {XCurve {keys [<at 0 1>,<at 1 5>]}, YCurve {keys [<at 0 1>,<at 1 1>]}}}, ScaleCurve=<1 1>))
   &CameraOffset(Camera Offset Amount=25)
  ]
  rendererClasses [&NiagaraSpriteRendererProperties]
  renderers [{rendererClass=&NiagaraSpriteRendererProperties}]
 }
 Light {
  emitterUpdateScript [
   &EmitterState(
    Inactive Response=::NewEnumerator1
    Life Cycle Mode=::NewEnumerator1
    Loop Behavior=::NewEnumerator1
    Loop Duration Mode=::NewEnumerator1
    UseLoopDelay=false
   )
   &SpawnBurst_Instantaneous(Spawn Count=2, Use Spawn Probability=false, UseLoopCountLimit=false)
   &SpawnRate(SpawnRate=20, Use Spawn Probability=false) {enabled=false}
  ]
  particleSpawnScript [
   &FN_SetSpawnGroup(Number of Spawn Groups=2)
   &InitializeParticle(
    Lifetime=&RandomRangeFloat(Minimum=0.2, Maximum=0.4, Evaluation Type=::NewEnumerator1)
    Module.Write Color=true
    Module.Write Lifetime=true
    Module.Write Mass=true
    Module.Write Material Random=true
    Module.Write Position=false
    Module.Write Scale=false
    Module.Write SpriteRotation=false
    Module.Write SpriteSize=true
    Sprite Size=<1 1>
   )
   /Niagara/Modules/Spawn/Location/SphereLocation(MaskSpawnGroup=false, Module.Override Randomness=false, Sphere Radius=12, Surface Only=true) {enabled=false}
  ]
  particleUpdateScript [
   &ParticleState(Kill Particles When Lifetime Has Elapsed=false, Let Infinitely Lived Particles Die When Emitter Deactivates=true)
   &SkeletalMeshLocation(
    Bone / Socket Index=$Particles.SpawnGroup
    Bone Sampling Mode=::NewEnumerator4
    Inherit Velocity=false
    MaskSpawnGroup=false
    Skeletal Mesh=$User.SkeletalMesh
   )
   &FN_Light_Attributes(
    Color=<rgba 1 0.479 0.1129 1>
    Create Color Attribute=true
    Create Exponent Attribute=true
    Create Radius Attribute=true
    Exponent=6
    Radius=80
    Scale Color=9
   )
  ]
  rendererClasses [&NiagaraSpriteRendererProperties, &NiagaraLightRendererProperties]
  renderers [
   {rendererClass=&NiagaraSpriteRendererProperties}
   {rendererClass=&NiagaraLightRendererProperties, rendererIndex=1}
  ]
 }
 Roman_Numerals {
  emitterUpdateScript [
   &EmitterState(
    Inactive Response=::NewEnumerator1
    Life Cycle Mode=::NewEnumerator1
    Loop Duration=&RandomRangeFloat(Minimum=1, Maximum=3, Evaluation Type=::NewEnumerator1)
    UseLoopDelay=true
   )
   &SpawnBurst_Instantaneous(Spawn Count=&RandomRangeInt(Minimum=10, Maximum=20), Use Spawn Probability=false)
   &SpawnRate(Spawn Probability=0.4, SpawnRate=15, Use Spawn Probability=true)
  ]
  particleSpawnScript [
   /SetVariables(Particles.SpriteRotationRate=&RandomRangeFloat(Minimum=0.75, Maximum=1, Evaluation Type=::NewEnumerator1)) {isSetParameters=true}
   &FN_SetSpawnGroup(Number of Spawn Groups=2)
   &InitializeParticle(
    Lifetime=&RandomRangeFloat(Minimum=0.25, Maximum=1.25, Evaluation Type=::NewEnumerator1)
    Module.Write Color=true
    Module.Write Lifetime=true
    Module.Write Mass=true
    Module.Write Material Random=true
    Module.Write Position=false
    Module.Write Scale=false
    Module.Write SpriteRotation=true
    Module.Write SpriteSize=true
    Sprite Rotation=&NormalizedAngleToDegrees(NormalizedAngle=&RandomRangeFloat(Minimum=-0.1, Maximum=0.1, Evaluation Type=::NewEnumerator1))
    Sprite Size=&Vector2DFromFloat(Value=&RandomRangeFloat(Minimum=5, Maximum=10, Evaluation Type=::NewEnumerator1))
   )
   &SkeletalMeshLocation(
    Bone / Socket Index=&RandomRangeInt(Minimum=0, Maximum=3, Randomness Mode=::NewEnumerator1, Random Seed=$Emitter.LoopCount, Fixed Random Seed=true)
    Bone Sampling Mode=::NewEnumerator4
    Inherit Velocity=false
    MaskSpawnGroup=false
    Skeletal Mesh=$User.SkeletalMesh
   ) {enabled=false}
   &SkeletalMeshLocation(
    Apply Sampled Orientation to Mesh Particles=true
    Bone / Socket Index=$Particles.SpawnGroup
    Bone Sampling Mode=::NewEnumerator4
    Inherit Velocity=false
    Interpolate Mesh Sampling=false
    MaskSpawnGroup=false
    Skeletal Mesh=$User.SkeletalMesh
   ) {module=SkeletalMeshLocation001}
   &FN_SkydivePositionOffset(
    Bone Orientation:Quat4f=$Output.SkeletalMeshLocation001.SampledRotation
    Bone Position:Vector3f=$Output.SkeletalMeshLocation001.SampledPosition
    Offset:NiagaraFloat=&SkydiveHandsFeetFloatModifier(Hands:NiagaraFloat=0, Feet:NiagaraFloat=0)
   )
   &InheritVelocity(Module.Limit Speed=false, Velocity Limit=0, Velocity Scale=<-0.075 -0.075 -0.075>)
   &Color(Color=&MakeLinearColorFromVectorAndFloat("Vector (RGB)"=<1 1 1>, "Float (Alpha)"=&ClampFloat(Float=1)))
   &AddVelocity(Velocity=&RandomVector(Vector Scale=100))
  ]
  particleUpdateScript [
   &ParticleState()
   &Drag(Drag=4, Use Linear Drag=true, Use Rotational Drag=false)
   &SolveForcesAndVelocity(Clamp Velocity=false, Limit Acceleration=false)
   &ScaleColor(
    Scale Alpha=&FloatFromCurve(FloatCurve=DataInterface() {propertyValues {Curve {keys [
         <at 0.102 0.0028 RCIM_Cubic>
         {arriveTangent=-0.0032, interpMode=RCIM_Cubic, leaveTangent=-0.0032, time=0.2344, value=1}
         <at 1 0 RCIM_Cubic>
        ]}}})
    Scale RGB=&VectorFromCurve(VectorCurve=DataInterface() {propertyValues {
       XCurve {keys [<at 0.2225 50 RCIM_Cubic>, {arriveTangent=-4, leaveTangent=-4, time=0.3579, value=5}, <at 1 1>]}
       YCurve {keys [<at 0.2225 18.5417 RCIM_Cubic>, {arriveTangent=-2.5702, leaveTangent=-2.5702, time=0.3579, value=2.9056}, <at 1 0.3354>]}
       ZCurve {keys [<at 0.2225 0.8824 RCIM_Cubic>, {arriveTangent=-0.1292, leaveTangent=-0.1292, time=0.3579, value=0.1607}, <at 1 0.0315>]}
      }})
    ScaleA=true
    ScaleRGB=true
   )
   &SpriteRotationRate(Rotation Rate=&NormalizedAngleToDegrees(NormalizedAngle=$Particles.SpriteRotationRate))
   /Niagara/Modules/CascadeConversion/CascadeConversion_Orbit(
    Offset:Vector3f=&VectorFromCurve(VectorCurve=DataInterface() {propertyValues {
       XCurve {keys [<at 0 0 RCIM_Cubic>,<at 1 0 RCIM_Cubic>]}
       YCurve {keys [{arriveTangent=26.7068, interpMode=RCIM_Cubic, time=0, value=19.9685}, <at 1 40 RCIM_Cubic>]}
       ZCurve {keys [<at 0 0 RCIM_Cubic>,<at 1 0 RCIM_Cubic>]}
      }})
    Rotation:Vector3f=&RandomRangeVector(Minimum=<0 0 0>, Maximum=<1 1 1>)
   ) {enabled=false}
   /Niagara/Modules/CascadeConversion/CascadeConversion_SolveOrbit() {enabled=false}
   &ScaleVelocity(Velocity Scale=&VectorFromFloat(Value=&FloatFromCurve(FloatCurve=DataInterface() {propertyValues {Curve {keys [
          <at 0.1013 0.1191 RCIM_Cubic>
          {arriveTangent=3.6566, interpMode=RCIM_Cubic, leaveTangent=3.6566, time=0.2022, value=2.4995}
          {arriveTangent=-1.8796, interpMode=RCIM_Cubic, leaveTangent=-1.8796, time=0.4222, value=1.2926}
          <at 1 1 RCIM_Cubic>
         ]}}}))) {enabled=false}
   &ScaleSpriteSize(Scale Factor=&Vector2DFromFloat(Value=&FloatFromCurve(FloatCurve=DataInterface() {propertyValues {Curve {keys [
          <at 0.0113 0.4968 RCIM_Cubic>
          {arriveTangent=3.784, interpMode=RCIM_Cubic, leaveTangent=3.784, time=0.0978, value=1.5123}
          {arriveTangent=-0.5659, interpMode=RCIM_Cubic, leaveTangent=-0.5659, time=0.2277, value=1.3156}
          <at 1 1.0017>
         ]}}})))
   &CameraOffset(Camera Offset Amount=-25)
  ]
  rendererClasses [&NiagaraSpriteRendererProperties]
  renderers [{rendererClass=&NiagaraSpriteRendererProperties}]
 }
 Sprites {
  bEnabled=false
  emitterUpdateScript [
   &EmitterState()
   &SpawnBurst_Instantaneous(Spawn Count=4, Spawn Probability=HlslExpression() {hlslExpression="Emitter.LoopCount == 0"}, Use Spawn Probability=true) {enabled=false}
   &SpawnRate(SpawnRate=8, Use Spawn Probability=false)
  ]
  particleSpawnScript [
   &FN_SetSpawnGroup()
   &InitializeParticle(
    Lifetime=&RandomRangeFloat(Minimum=1.2, Maximum=2, Evaluation Type=::NewEnumerator1)
    Module.Write Color=true
    Module.Write Lifetime=true
    Module.Write Mass=true
    Module.Write Material Random=true
    Module.Write Position=false
    Module.Write Scale=false
    Module.Write SpriteRotation=false
    Module.Write SpriteSize=false
    Position=&SimulationPosition()
   )
   &AddVelocity(Velocity=&RandomRangeVector(Minimum=<-10 -10 50>, Maximum=<10 10 100>, Evaluation Type=::NewEnumerator1))
  ]
  particleUpdateScript [
   &ParticleState()
   &SkeletalMeshLocation(
    Bone / Socket Index=$Particles.SpawnGroup
    Bone Sampling Mode=::NewEnumerator4
    Inherit Velocity=false
    Interpolate Mesh Sampling=false
    MaskSpawnGroup=false
    Skeletal Mesh=$User.SkeletalMesh
   )
   &FN_Sprite_Attributes(
    Create SpriteSize Attribute=true
    SpriteSize=&Vector2DFromFloat(Value=&FloatFromCurve(FloatCurve=DataInterface() {propertyValues {Curve {keys [<at 0 1>,<at 1 5>]}}}, Scale Curve=33))
    Write to Module Namespace=true
   )
   &FN_Sprite_Attributes(
    Camera Offset=25
    Color=&MakeLinearColorFromVectorAndFloat(
     "Vector (RGB)"=&VectorFromCurve(VectorCurve=DataInterface() {propertyValues {XCurve {keys [<at 0 30>,<at 1 20>]}, YCurve {keys [<at 0 25.2878>,<at 1 13.7166>]}, ZCurve {keys [<at 0 9.6875>,<at 1 0.4988>]}}})
     "Float (Alpha)"=&ClampFloat(Float=&FloatFromCurve(FloatCurve=DataInterface() {propertyValues {Curve {keys [
           <at -0.0003 0 RCIM_Cubic>
           {arriveTangent=1.8811, interpMode=RCIM_Cubic, leaveTangent=1.8811, time=0.1829, value=0.9903}
           {arriveTangent=-1, leaveTangent=-1, time=0.5313, value=1}
           <at 1 0>
          ]}}}))
    )
    Create Camera Offset Attribute=true
    Create Color Attribute=true
    Create Sprite Rotation Attribute=true
    Create SpriteSize Attribute=true
    SpriteRotation:NiagaraFloat=&RandomRangeFloat(Minimum=0, Maximum=360)
    SpriteSize=&Vector2DFromCurve(
     Vector2Curve=DataInterface() {propertyValues {XCurve {keys [<at 0 1>,<at 1 5>]}, YCurve {keys [<at 0 1>,<at 1 1>]}}}
     ScaleCurve=&RandomRangeVector2D(Minimum=<15 15>, Maximum=<25 25>)
    )
    Write to Module Namespace=true
   ) {module=FN_Sprite_Attributes001}
   &FN_Sprite_Attributes(
    Color=&MakeLinearColorFromVectorAndFloat(
     "Vector (RGB)"=&VectorFromCurve(VectorCurve=DataInterface() {propertyValues {XCurve {keys [<at 0 30>,<at 1 20>]}, YCurve {keys [<at 0 25.2878>,<at 1 13.7166>]}, ZCurve {keys [<at 0 9.6875>,<at 1 0.4988>]}}})
     "Float (Alpha)"=&ClampFloat(Float=&FloatFromCurve(FloatCurve=DataInterface() {propertyValues {Curve {keys [
           <at -0.0003 0 RCIM_Cubic>
           {arriveTangent=1.8811, interpMode=RCIM_Cubic, leaveTangent=1.8811, time=0.1829, value=0.9903}
           {arriveTangent=-1, leaveTangent=-1, time=0.5313, value=1}
           <at 1 0>
          ]}}}))
    )
    Create Color Attribute=true
    Create SpriteSize Attribute=true
    SpriteSize=&Vector2DFromFloat(Value=&FloatFromCurve(FloatCurve=DataInterface() {propertyValues {Curve {keys [<at 0 1>,<at 1 5>]}}}, Scale Curve=25))
    Write to Module Namespace=true
   ) {module=FN_Sprite_Attributes002}
   &FN_Sprite_Attributes(
    Color=&MakeLinearColorFromVectorAndFloat(
     "Vector (RGB)"=&VectorFromCurve(VectorCurve=DataInterface() {propertyValues {XCurve {keys [<at 0 1>,<at 1 1>]}, YCurve {keys [<at 0 1>,<at 1 1>]}, ZCurve {keys [<at 0 1>,<at 1 1>]}}})
     "Float (Alpha)"=&ClampFloat(Float=&FloatFromCurve(FloatCurve=DataInterface() {propertyValues {Curve {keys [{arriveTangent=0.0516, interpMode=RCIM_Cubic, time=-0.003, value=0}, <at 0.2978 0.2015 RCIM_Cubic>, <at 1 0>]}}}))
    )
    Create Color Attribute=true
    Create SpriteSize Attribute=true
    Create Visibility Attribute=true
    SpriteSize=&Multiply_Vector2(
     A=&RandomRangeVector2D(Minimum=<48 48>, Maximum=<50 50>)
     B=&Vector2DFromFloat(Value=&FloatFromCurve(FloatCurve=DataInterface() {propertyValues {Curve {keys [
           {arriveTangent=19.8573, interpMode=RCIM_Cubic, time=0, value=0.3202}
           {arriveTangent=0.5332, interpMode=RCIM_Cubic, leaveTangent=0.5332, time=0.1513, value=0.7545}
           {arriveTangent=-0.4802, interpMode=RCIM_Cubic, leaveTangent=-0.4802, time=0.8198, value=0.7573}
           {interpMode=RCIM_Cubic, leaveTangent=-14.9405, time=1.0022, value=0.3459}
          ]}}}))
    )
    Visibility Tag:NiagaraInt32=&RandomRangeInt(Minimum=0, Maximum=1)
    Write to Module Namespace=true
   ) {module=FN_Sprite_Attributes003}
   &FN_Sprite_Attributes(
    Color=&MakeLinearColorFromVectorAndFloat(
     "Vector (RGB)"=<1 0.4904 0.0776>
     "Float (Alpha)"=&ClampFloat(Float=&FloatFromCurve(FloatCurve=DataInterface() {propertyValues {Curve {keys [{arriveTangent=0.0516, interpMode=RCIM_Cubic, time=-0.003, value=0}, <at 0.2978 0.2468 RCIM_Cubic>, <at 1 0>]}}}))
    )
    Create Color Attribute=true
    Create SpriteSize Attribute=true
    SpriteSize=&RandomRangeVector2D(Minimum=<50 50>, Maximum=<70 70>)
    Write to Module Namespace=true
   ) {module=FN_Sprite_Attributes004}
   &DynamicMaterialParameters(
    Index 0 Param 1=1
    Index 0 Param 2=&RandomRangeFloat(Minimum=0, Maximum=10, Evaluation Type=::NewEnumerator1)
    Index 0 Param 3=&FloatFromCurve(FloatCurve=DataInterface() {propertyValues {Curve {keys [
         {arriveTangent=0.5989, interpMode=RCIM_Cubic, time=0, value=-0.1494}
         {arriveTangent=0.2988, interpMode=RCIM_Cubic, leaveTangent=0.2988, time=0.5, value=0}
         {interpMode=RCIM_Cubic, leaveTangent=0.4853, time=1, value=0.1494}
        ]}}})
    Index 0 Param 4=&FloatFromCurve(FloatCurve=DataInterface() {propertyValues {Curve {keys [
         {arriveTangent=-5.637, interpMode=RCIM_Cubic, time=0, value=0.9968}
         {arriveTangent=-0.9983, interpMode=RCIM_Cubic, leaveTangent=-0.9983, time=0.3196, value=0.0047}
         <at 0.9984 0 RCIM_Cubic>
        ]}}})
    Index 1 Param 1=1
    Index 1 Param 3=0.1
    Index 1 Param 4=&FloatFromCurve(FloatCurve=DataInterface() {propertyValues {Curve {keys [
         {arriveTangent=-5.637, interpMode=RCIM_Cubic, time=0, value=0.9968}
         {arriveTangent=-0.9983, interpMode=RCIM_Cubic, leaveTangent=-0.9983, time=0.3196, value=0.0047}
         <at 0.9984 0 RCIM_Cubic>
        ]}}})
    Module.Param0WriteEnabled=true
    Module.Param0WriteEnabled1=true
    Module.Param1WriteEnabled=true
    Module.Param1WriteEnabled1=true
    Module.Param2WriteEnabled=true
    Module.Param2WriteEnabled1=true
    Module.Param3WriteEnabled=true
    Module.Param3WriteEnabled1=true
    Write Parameter Index 1=true
   )
   &SolveForcesAndVelocity(Clamp Velocity=false, Limit Acceleration=false)
  ]
  rendererClasses [&NiagaraSpriteRendererProperties]
  renderers [
   {rendererClass=&NiagaraSpriteRendererProperties}
   {rendererClass=&NiagaraSpriteRendererProperties, rendererIndex=1}
   {rendererClass=&NiagaraSpriteRendererProperties, rendererIndex=2}
   {rendererClass=&NiagaraSpriteRendererProperties, rendererIndex=3}
   {rendererClass=&NiagaraSpriteRendererProperties, rendererIndex=4}
  ]
 }
 SubUV {
  emitterUpdateScript [
   &EmitterState(Life Cycle Mode=::NewEnumerator1, Loop Duration=1, UseLoopDelay=false)
   &SpawnRate(SpawnRate=20, Use Spawn Probability=false)
   &SpawnBurst_Instantaneous(Spawn Count=4, Use Spawn Probability=false) {enabled=false}
  ]
  particleSpawnScript [
   &FN_SetSpawnGroup()
   &InitializeParticle(
    Lifetime=&RandomRangeFloat(Minimum=1, Maximum=1.5, Evaluation Type=::NewEnumerator1)
    Module.Write Color=true
    Module.Write Lifetime=true
    Module.Write Mass=true
    Module.Write Material Random=true
    Module.Write Position=false
    Module.Write Scale=false
    Module.Write SpriteRotation=true
    Module.Write SpriteSize=true
    Sprite Rotation=&NormalizedAngleToDegrees(NormalizedAngle=&RandomRangeFloat(Minimum=0, Maximum=1, Evaluation Type=::NewEnumerator1))
    Sprite Size=&Vector2DFromFloat(Value=&SkydiveHandsFeetFloatModifier(Hands:NiagaraFloat=7, Feet:NiagaraFloat=5))
   )
   &AddVelocity(Velocity=&RandomRangeVector(Minimum=<-10 -10 50>, Maximum=<10 10 100>, Evaluation Type=::NewEnumerator1)) {enabled=false}
  ]
  particleUpdateScript [
   &ParticleState()
   &SolveForcesAndVelocity(Clamp Velocity=false, Limit Acceleration=false)
   &SkeletalMeshLocation(
    Apply Sampled Orientation to Mesh Particles=true
    Bone / Socket Index=$Particles.SpawnGroup
    Bone Sampling Mode=::NewEnumerator4
    Inherit Velocity=false
    Interpolate Mesh Sampling=false
    MaskSpawnGroup=false
    Skeletal Mesh=$User.SkeletalMesh
   ) {module=SkeletalMeshLocation001}
   &Color(Color=&MakeLinearColorFromVectorAndFloat(
     "Vector (RGB)"=&VectorFromCurve(
      VectorCurve=DataInterface() {propertyValues {XCurve {keys [<at 0 1>,<at 1 0>]}, YCurve {keys [<at 0 1>,<at 1 0>]}, ZCurve {keys [<at 0 1>,<at 1 0>]}}}
      Scale Curve=<0.1 0.1 0.1>
     )
     "Float (Alpha)"=&ClampFloat(
      Float=&FloatFromCurve(FloatCurve=DataInterface() {propertyValues {Curve {keys [{arriveTangent=-1, time=0, value=1}, <at 1 0>]}}})
      Min=0.5
     )
    ))
   &ScaleSpriteSize(Scale Factor=&Vector2DFromFloat(Value=&FloatFromCurve(FloatCurve=DataInterface() {propertyValues {Curve {keys [<at 0 1>,<at 1 5>]}}}, Scale Curve=2)))
   &SubUVAnimation(Number Of Frames=49)
   /SetVariables(Particles.DynamicMaterialParameter:Vector4f=&MakeVector4_FromLinearColor(LinearColor=&ColorFromCurve(
      ColorCurve=DataInterface() {propertyValues {
        AlphaCurve {keys [<at 0 1 RCIM_Cubic>,<at 1 1 RCIM_Cubic>]}
        BlueCurve {keys [<at 0 0>,<at 0.4991 0>,<at 1 0.1066 RCIM_Cubic>]}
        GreenCurve {keys [<at 0 0>,<at 0.4991 1>,<at 1 0 RCIM_Cubic>]}
        RedCurve {keys [<at 0 1>,<at 0.4991 0>,<at 1 1 RCIM_Cubic>]}
       }}
      CurveIndex=HlslExpression() {hlslExpression="Modulo(Particles.NormalizedAge*2 + Particles.MaterialRandom, 1.0)"}
     ))) {isSetParameters=true}
  ]
  rendererClasses [&NiagaraSpriteRendererProperties]
  renderers [{rendererClass=&NiagaraSpriteRendererProperties}]
 }
}
summary {
 systemName=NS_Contrail_Sands
 emitters [
  {emitterName=DarkGradient, rendererClasses [&NiagaraSpriteRendererProperties]}
  {emitterName=SubUV, rendererClasses [&NiagaraSpriteRendererProperties]}
  {emitterName=FogCenter, rendererClasses [&NiagaraSpriteRendererProperties]}
  {emitterName=Fog, rendererClasses [&NiagaraSpriteRendererProperties]}
  {emitterName=Clock, rendererClasses [&NiagaraSpriteRendererProperties]}
  {emitterName=Clock001, rendererClasses [&NiagaraSpriteRendererProperties]}
  {emitterName=Light, rendererClasses [&NiagaraSpriteRendererProperties, &NiagaraLightRendererProperties]}
  {emitterName=Roman_Numerals, rendererClasses [&NiagaraSpriteRendererProperties]}
  {bEnabled=false, emitterName=Sprites, rendererClasses [&NiagaraSpriteRendererProperties]}
  {emitterName=FN_Ribbons, rendererClasses [&NiagaraRibbonRendererProperties]}
 ]
 userVariables [
  User.bIsFrontEnd {type=NiagaraBool}
  User.bIsFrontEndPreview {type=NiagaraBool}
  User.ParaGlideLeanAlpha {type=NiagaraFloat, defaultValue=0, description=""}
  User.RibbonIDOffset {type=NiagaraInt32}
  User.SkeletalMesh {
   type=NiagaraDataInterfaceSkeletalMesh
   defaultValue {
    struct=NiagaraExt_VariableValue_DataInterface
    value {dataInterface=&self:NiagaraDataInterfaceSkeletalMesh_0}
   }
  }
 ]
}
